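; Quest gameplay ability that credits the player for swinging from a tree with a grapple, ignoring brief attaches.
; After event activation and a short delay, it starts a looping timer and ends if the MovementMode.Swinging tag is removed.
; Timer ticks accumulate swing time, sending one quest stat event at a threshold before ending the ability.
; Ending resets the elapsed time and sent flag and clears the timer.
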(event Ability|EventActivateAbilityFromEvent
  (params (EventData :type Gameplay%20Event%20Data%20Structure%20%28by%20ref%29))
  ;[0 0 0] Activate ability when player attaches to a tree, make sure they are swinging
  (|CallParentFunction EventData)
  (Utilities|FlowControl|Delay 0.2)
  (Ability|Tasks|WaitGameplayTagRemove "(TagName=\"MovementMode.Swinging\")" 0 true
    (:then
      (bind _returnvalue (Utilities|Time|SetTimerbyEvent (EventDispatchers|CreateEvent :!event Custom|Swinging) (Variables|Default|GetTimeInterval) true))
      (Variables|Default|SetTimerHandle _returnvalue))
    (:Removed
      (Ability|EndAbility))))

(event Custom|Swinging
  ;[0 0 0] If the player swings for at least a couple seconds, send a quest update (to prevent just clicking on a tree for credit)
  (if (GameplayTags|HasMatchingGameplayTag (Variables|GAVars|Data|GetOwningPlayerPawn) "(TagName=\"MovementMode.Swinging\")")
    (bind _output_get (Variables|Default|SetTimeElapsed (+ (Variables|Default|GetTimeElapsed) (Variables|Default|GetTimeInterval))))
    (if (>= _output_get 1.5)
      (if (not (Variables|Default|GetbStatEventSent))
        (Development|PrintString "Sending stat event" true true "(R=1.000000,G=0.000000,B=0.382610,A=1.000000)" 3.0)
        (Quest|SendComplexCustomStatEvent (Variables|GAVars|Data|GetOwningPlayerQuestManager) 0 (Variables|Default|GetQuestSourceTag) "(GameplayTags=)" 1)
        (Variables|Default|SetbStatEventSent true)
        (Ability|EndAbility)))
    (else
      (Ability|EndAbility))))

(event Ability|EventOnEndAbility
  (params (bWasCancelled :type Boolean))
  ;[0 0 0] End ability
  (|CallParentFunction bWasCancelled)
  (Development|PrintString "Ability ending..." true true "(R=1.000000,G=0.000000,B=0.382610,A=1.000000)" 3.0)
  (Variables|Default|SetTimeElapsed 0.0)
  (Variables|Default|SetbStatEventSent false)
  (if (Utilities|Time|IsValidTimerHandle (Variables|Default|GetTimerHandle))
    (Utilities|Time|ClearandInvalidateTimerbyHandle (Variables|Default|GetTimerHandle))))
